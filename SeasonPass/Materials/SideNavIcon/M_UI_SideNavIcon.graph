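# Translucent UI material for a Season Pass side navigation icon, with free/paid and selected states.
# An IsPaid scalar blends between the alpha channels of the FreeIcon and PaidIcon textures, both defaulting to the Fortnite Crew icon, to form the icon shape.
# A separate Selected scalar blends the tint from white to near-black.
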Material /SeasonPass/Materials/SideNavIcon/M_UI_SideNavIcon
graph {
 properties {
  bAutomaticallySetUsageInEditor=false
  bUsedWithStaticMesh=false
  blendMode=BLEND_Translucent
  materialDomain=MD_UI
  usedShadingModels=Default Lit
 }
 nodes [
  LinearInterpolate(
   A=TextureSampleParameter2D() {parameterName=FreeIcon, texture=/Game/UI/SocialPanel/Art/Textures/T_UI_Icon_FortniteCrew}.A
   B=TextureSampleParameter2D() {parameterName=PaidIcon, texture=&T_UI_Icon_FortniteCrew}.A
   Alpha=ScalarParameter() {parameterName=IsPaid, sliderMax=1}
  )
  LinearInterpolate(
   A=Constant3Vector() {constant=<rgba 1 1 1 0>}
   B=Constant3Vector() {constant=<rgba 0.0048 0.004 0.0116 1>}
   Alpha=ScalarParameter() {parameterName=Selected, sliderMax=1}
  )
 ]
}
